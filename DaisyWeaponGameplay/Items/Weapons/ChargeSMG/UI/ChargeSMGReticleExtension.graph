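# Charge SMG reticle extension widget that shows clip ammo and the projected cost of a charged burst on a material-driven bar.
# While charging, the RoundsPerBurst curve converts charge to bullets, lowering the material's Progress, while Delta Progress tracks current ammo.
# Weapon and pawn events drive ammo updates, holster hiding and a reload animation scaled to reload time.
# One fade animation plays forward or reversed, with a timer fading the bar out unless secondary fire is held.
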WidgetBlueprint /DaisyWeaponGameplay/Items/Weapons/ChargeSMG/UI/ChargeSMGReticleExtension
animations [
 FadeOut {
  binding_count=1
  length_seconds=0.4969
  object=&self:FadeOut
  sequence {bindings [Overlay_0 {
     id=#1
     possessed_class=Overlay
     tracks [{
       type=MovieSceneFloatTrack
       display_name=Render Opacity
       sections [{type=MovieSceneFloatSection, channels [None {type=MovieSceneScriptingFloatChannel, keys [[0,1],[9.937,0]]}]}]
      }]
    }]}
 }
 Reload {
  binding_count=1
  length_seconds=0.9977
  object=&self:Reload
  sequence {bindings [Ammo_Progress {
     id=#2
     possessed_class=Image
     tracks [{
       type=MovieSceneWidgetMaterialTrack
       sections [{
         type=MovieSceneParameterSection
         channels [
          OverridePercent {type=MovieSceneScriptingFloatChannel, keys [[0,0],[19.954,1]]}
          SegmentOverride {type=MovieSceneScriptingFloatChannel, const [1,2]}
          Percent {type=MovieSceneScriptingFloatChannel, const [0,1]}
         ]
        }]
      }]
    }]}
 }
]
graphs [
 BindEvents {nodes [
   BindEvents() {is=title K2Node_FunctionEntry}
   > UnbindEvents()
   > n1 = Sequence()
   > n2 = Set EquippedWeapon(EquippedWeapon:Fort Weapon Object Reference=GetEquippedWeapon!(self:Fort Weapon Reticle Extension Widget Base Object Reference) -> (ReturnValue:Fort Weapon Object Reference)) -> (Output_Get:Fort Weapon Object Reference)
   > Bind Event to On Local Ammo Changed(self:Fort Weapon Object Reference=@n2, Delegate=Create Event()) {is=title K2Node_AddDelegate}
   n3 = Cast To FortPawn!(
    execute=@n1.then_1
    Object=Get Instigator!(self:Actor Object Reference=Get EquippedWeapon() -> (EquippedWeapon:Fort Weapon Object Reference)) -> (Instigator:Pawn Object Reference)
   ) -> (then, CastFailed, AsFort Pawn:Fort Pawn Object Reference) {is=title K2Node_DynamicCast}
   > n4 = Set InstigatorPawn(InstigatorPawn:Fort Pawn Object Reference=@n3) -> (Output_Get:Fort Pawn Object Reference)
   > Bind Event to On Holstered Event(self:Fort Pawn Object Reference=@n4, Delegate=Create Event()) {is=title K2Node_AddDelegate}
  ]}
 ConvertChargeToBullets {nodes [
   n1 = ConvertChargeToBullets!() -> (then, ChargePercent:double=~) {is=title K2Node_FunctionEntry}
   n2 = GetFloatValue!(self:Curve Float Object Reference=Get RoundsPerBurst() -> (RoundsPerBurst:Curve Float Object Reference), InTime:float=@n1) -> (ReturnValue:float)
   > Return Node!(execute, Bullets:int=SelectInt(A=Floor(A=@n2) {is=title K2Node_CallFunction}, B=0, bPickA="float > float"(A=@n1)))
  ]}
 EventGraph {nodes [
   EventOutro() {is=title K2Node_CustomEvent}
   > n1 = Branch(Condition=Get IsHoldingSecondaryFire() -> (IsHoldingSecondaryFire:bool))
   n2 = PlayFadeAnimation!(execute=@n1.else, self, DoFadeIn:bool=true, DoSnapToEnd:bool=true) -> (then)
   n3 = EventFadeOut() {is=title K2Node_CustomEvent}
   n4 = Set Timer by Event(execute=@n2.then, Delegate=@n3, Time=Get TimeToOutro() -> (TimeToOutro:double), bLooping=false)
   > Set FadeOutHandle(FadeOutHandle:Timer Handle Structure=@n4) -> (Output_Get:Timer Handle Structure)
   n5 = EventReloadFinished() {is=title K2Node_CustomEvent}
   > n6 = EventOutro() {is=title K2Node_CallFunction}
   n7 = EventHideReloadParams() {is=title K2Node_CustomEvent}
   Set Timer for Next Tick by Event!(execute=@n6.then, Delegate:Delegate=@n7) -> (then, ReturnValue:Timer Handle Structure) {is=title K2Node_CallFunction}
   HandleReloadCanceled() {is=title K2Node_CustomEvent}
   > StopAnimation(InAnimation=Get Reload() -> (Reload:Widget Animation Object Reference))
   > UpdateAmmoDisplay()
   Event Destruct()
   > UnbindEvents()
   Event Construct()
   > BindEvents() {is=title K2Node_CallFunction}
   > SetupWeapon()
   > EventOutro()
   n8 = Get Reload()
   > Set ReloadAnimationLength(ReloadAnimationLength:double=float - float(A=GetEndTime(self=@n8), B=GetStartTime!(self:Widget Animation Object Reference=@n8) -> (ReturnValue:float)) {is=title K2Node_CallFunction}) -> (Output_Get:double)
   > BindToAnimationFinished(Animation=Get Reload(), Delegate=@n5)
   n9 = Event OnWeaponReloadStarted!() -> (OutputDelegate, then, ReloadTime:float, ReloadType:EFortWeaponReloadType Enum=ReloadWholeClip) {is=title K2Node_Event}
   > HandleReload!(execute, self, ReloadTime:double=@n9.ReloadTime, ReloadType:EFortWeaponReloadType Enum=@n9.ReloadType) -> (then)
   Event OnWeaponReloadCancelled() {is=title K2Node_Event}
   > HandleReloadCanceled() {is=title K2Node_CallFunction}
   n10 = Event OnTargetingChanged!() -> (OutputDelegate, then, bIsADS:bool) {is=title K2Node_Event}
   > HandleTargeting!(execute, self, bIsTargeting:bool=@n10.bIsADS) -> (then)
   Event OnAmmoClipAttributeChanged!() -> (OutputDelegate, then, AmmoClipValue:float) {is=title K2Node_Event}
   > UpdateAmmoSegments()
   HideReloadParams(execute=@n7.then)
   PlayFadeAnimation(execute=@n11.else, DoFadeIn=false, DoSnapToEnd=false)
   n11 = Branch(execute=@n3.then, Condition=Get IsHoldingSecondaryFire())
  ]}
 GetClipPercentFull {nodes [
   GetClipPercentFull() {is=title K2Node_FunctionEntry}
   n1 = GetEquippedWeapon()
   n2 = GetBulletsPerClip!(self:Fort Weapon Object Reference=@n1) -> (ReturnValue:int=0)
   n3 = GetMagazineAmmoCount!(self:Fort Weapon Object Reference=@n1) -> (ReturnValue:int=0)
   > Return Node!(execute, ClipPercent:double=float / float(A="To Float (Integer)"(InInt=@n3), B="To Float (Integer)"(InInt=@n2)))
  ]}
 HandleAmmoChanged {nodes [
   n1 = HandleAmmoChanged!() -> (then, LocalCount:int, LocalRemaining:int) {is=title K2Node_FunctionEntry}
   n2 = GetBulletsPerClip(self=Get EquippedWeapon()) -> (ReturnValue=0)
   > n3 = Set NewAmmoPercent(NewAmmoPercent:double=float / float(A:int=@n1, B="To Float (Integer)"(InInt=@n2))) -> (Output_Get:double)
   > n4 = Branch(Condition="float > float"(A=@n3, B=Get CurrentAmmoPercent() -> (CurrentAmmoPercent:double)))
   > n5 = SetProgress!(execute, self, PercentAmmoRemaining:double=Get NewAmmoPercent() -> (NewAmmoPercent:double)) -> (then)
   Set CurrentAmmoPercent(execute=[@n5.then, @n4.else], CurrentAmmoPercent:double=Get NewAmmoPercent()) -> (Output_Get:double)
   > UpdateAmmoDisplay()
  ]}
 HandleReload {nodes [
   n1 = HandleReload!() -> (then, ReloadTime:double=~, ReloadType:EFortWeaponReloadType Enum=ReloadWholeClip) {is=title K2Node_FunctionEntry}
   > n2 = Branch(Condition="Equal (Enum)"(A:EFortWeaponReloadType Enum=@n1.ReloadType, B:EFortWeaponReloadType Enum=ReloadIndividualBullets))
   n3 = GetDynamicMaterial(execute=@n2.else, self=Get Ammo_Progress() -> (Ammo_Progress:Image Object Reference))
   > n4 = Get Scalar Parameter Value!(execute, self:Material Instance Dynamic Object Reference=@n3, ParameterName:Name=Delta Progress) -> (then, ReturnValue:float) {is=title K2Node_CallFunction}
   > Set CurrentAmmoPercent(CurrentAmmoPercent=@n4)
   n5 = "float > float"(A=@n1, B=0) {is=title K2Node_CallFunction}
   > PlayAnimation(
    InAnimation=Get Reload()
    PlaybackSpeed=float / float(A=1, B=Select(Option 0:double=1, Option 1:double=@n1, Index=@n5) -> (ReturnValue:double)) {is=title K2Node_CallFunction}
   )
   > PlayFadeAnimation(DoFadeIn=true, DoSnapToEnd=true)
  ]}
 HandleTargeting {nodes [
   n1 = HandleTargeting!() -> (then, bIsTargeting:bool=~) {is=title K2Node_FunctionEntry}
   > Branch(Condition=@n1)
   > EventOutro()
  ]}
 HideReloadParams {nodes [
   n1 = GetDynamicMaterial(execute=@n3.else, self=Get Ammo_Progress())
   SetScalarParameterValue(self=@n1, ParameterName=OverridePercent, Value=0)
   > n2 = SetScalarParameterValue(self=@n1, ParameterName=SegmentOverride, Value=0)
   SetScalarParameterValue(execute=[@n2.then, @n1.then], self=@n1, ParameterName=Delta Progress, Value=Get CurrentAmmoPercent())
   HideReloadParams() {is=title K2Node_FunctionEntry}
   > n3 = Branch(Condition=IsAnimationPlaying(InAnimation=Get Reload()))
  ]}
 On Weapon Holstered {nodes [
   On Weapon Holstered() {is=title K2Node_FunctionEntry}
   > SetDisplayedChargePercent!(execute, self, NewChargePercent:double=0) -> (then)
   > PlayFadeAnimation(DoFadeIn=false, DoSnapToEnd=true)
  ]}
 OnUpdateChargeDisplay {nodes [
   n1 = OnUpdateChargeDisplay!() -> (then, ChargePercent:float) {is=title K2Node_FunctionEntry}
   n2 = GetEquippedWeapon()
   n3 = GetChargeToMinPercent!(self:Fort Weapon Object Reference=@n2) -> (ReturnValue:float)
   > Branch(Condition=AND Boolean(A=GetIsChargingWeapon!(self:Fort Weapon Object Reference=@n2) -> (ReturnValue:bool), B="float >= float"(A=@n3, B=1)))
   > SetDisplayedChargePercent(NewChargePercent=@n1)
   > n4 = Branch(Condition="Not Equal (Float)"(A=Get Cached Charge Percent() -> (Cached Charge Percent:double)))
   > PlayFadeAnimation(DoFadeIn=true, DoSnapToEnd=true)
   EventOutro(execute=@n4.else)
   Return Node!(execute)
  ]}
 PlayFadeAnimation {nodes [
   n1 = PlayFadeAnimation!() -> (then, DoFadeIn:bool=~, DoSnapToEnd:bool=~) {is=title K2Node_FunctionEntry}
   > Clear and Invalidate Timer by Handle(Handle=Get FadeOutHandle() -> (FadeOutHandle:Timer Handle Structure))
   n2 = IsAnimationPlaying(InAnimation=Get Fadeout() -> (Fadeout:Widget Animation Object Reference))
   n3 = Get IsReticleVisible() -> (IsReticleVisible:bool)
   > Branch(Condition=OR Boolean(A="Not Equal (Boolean)"(A=@n1, B=@n3) -> (ReturnValue=~), B=AND Boolean(A=@n1.DoSnapToEnd, B=@n2)))
   > n4 = Set IsReticleVisible(IsReticleVisible:bool=Get DoFadeIn() -> (DoFadeIn:bool)) -> (Output_Get:bool)
   > n5 = Branch(Condition=@n4)
   n6 = Get Fadeout()
   n7 = Select(Option 0=0, Option 1=GetEndTime(self=@n6), Index=Get DoSnapToEnd() -> (DoSnapToEnd:bool))
   > PlayAnimation(InAnimation=@n6, StartAtTime=@n7, PlayMode=Reverse)
   PlayAnimation(execute=@n5.else, InAnimation=@n6, StartAtTime=@n7)
  ]}
 SetDisplayedChargePercent {nodes [
   n1 = SetDisplayedChargePercent!() -> (then, NewChargePercent:double=~) {is=title K2Node_FunctionEntry}
   > n2 = Set Cached Charge Percent(Cached Charge Percent:double=@n1) -> (Output_Get:double)
   > n3 = ConvertChargeToBullets!(execute, self, ChargePercent:double=@n2) -> (then, Bullets:int=0) {is=title K2Node_CallFunction}
   n4 = GetEquippedWeapon()
   > n5 = Set NewBulletCount(NewBulletCount:int=int - int(A=GetMagazineAmmoCount(self=@n4) -> (ReturnValue=0), B=@n3)) -> (Output_Get:int=0)
   n6 = GetBulletsPerClip(self=GetEquippedWeapon()) -> (ReturnValue=0)
   > SetProgress(PercentAmmoRemaining=float / float(A:double="To Float (Integer)"(InInt=@n5), B="To Float (Integer)"(InInt=@n6)) {is=title K2Node_PromotableOperator})
  ]}
 SetProgress {nodes [
   n1 = SetProgress!() -> (then, PercentAmmoRemaining:double=~) {is=title K2Node_FunctionEntry}
   > n2 = GetDynamicMaterial(self=Get Ammo_Progress())
   > SetScalarParameterValue(self=@n2, ParameterName=Progress, Value=@n1)
  ]}
 SetupWeapon {nodes [
   SetupWeapon() {is=title K2Node_FunctionEntry}
   > Set Cached Charge Percent(Cached Charge Percent=-1)
   > UpdateAmmoSegments()
   > HandleAmmoChanged!(execute, self, LocalCount:int=GetMagazineAmmoCount(self=GetEquippedWeapon()) -> (ReturnValue=0), LocalRemaining:int=0) -> (then) {is=title K2Node_CallFunction}
   > SetProgress(PercentAmmoRemaining=Get CurrentAmmoPercent()) {is=title K2Node_CallFunction}
  ]}
 UnbindEvents {nodes [
   UnbindEvents() {is=title K2Node_FunctionEntry}
   > n1 = Sequence()
   > Is Valid(InputObject=Get EquippedWeapon())
   > Unbind Event from On Local Ammo Changed(self:Fort Weapon Object Reference=Get EquippedWeapon(), Delegate=Create Event()) {is=title K2Node_RemoveDelegate}
   > Set EquippedWeapon()
   Is Valid(exec=@n1.then_1, InputObject=Get InstigatorPawn() -> (InstigatorPawn:Fort Pawn Object Reference))
   > Unbind Event from On Holstered Event(self:Fort Pawn Object Reference=Get InstigatorPawn(), Delegate=Create Event()) {is=title K2Node_RemoveDelegate}
   > Set InstigatorPawn()
  ]}
 UpdateAmmoDisplay {nodes [
   UpdateAmmoDisplay() {is=title K2Node_FunctionEntry}
   > n1 = GetDynamicMaterial(self=Get Ammo_Progress())
   > SetScalarParameterValue(self=@n1, ParameterName=Delta Progress, Value=Get CurrentAmmoPercent())
   > EventOutro()
  ]}
 UpdateAmmoSegments {nodes [
   UpdateAmmoSegments() {is=title K2Node_FunctionEntry}
   > n1 = GetDynamicMaterial(self=Get Ammo_Progress())
   n2 = GetBulletsPerClip(self=GetEquippedWeapon()) -> (ReturnValue=0)
   n3 = int * int(A:int=@n2, B:int=4) -> (ReturnValue:int=0) {is=title K2Node_CommutativeAssociativeBinaryOperator}
   > SetScalarParameterValue(
    self=@n1
    ParameterName=Segments
    Value=Select(Option 0="To Float (Integer)"(InInt=@n3), Option 1=1, Index="Equal (Integer)"(A=@n2, B=1) {is=title K2Node_CallFunction})
   )
  ]}
]
variables [
 Cached Charge Percent {type=number, default=-1}
 CurrentAmmoPercent {type=number}
 EquippedWeapon {type=FortWeapon, default=null}
 FadeOutHandle {type=TimerHandle}
 InstigatorPawn {type=FortPawn, default=null}
 IsHoldingSecondaryFire {type=boolean}
 IsReticleVisible {type=boolean}
 ReloadAnimationLength {type=number}
 RoundsPerBurst {type=CurveFloat, default=/DaisyWeaponGameplay/Items/Weapons/ChargeSMG/ChargeSMG_RoundsPerBurst}
 TimeToOutro {type=number, default=1.5}
]
widgets {
 info {
  inheritedWidgetCount=1
  parentClass=/Script/FortniteUI.FortChargedWeaponReticleExtensionWidget
  rootWidgetClass=&Overlay
  widgetCount=3
 }
 widgets [
  Overlay_0 {widgetClassPath=&Overlay, widget=Overlay_0}
  Ammo_Progress {
   widgetClassPath=&Image
   bIsVariable=true
   parent=Overlay_0
   widget=Ammo_Progress
   properties {
    visibility=HitTestInvisible
    brush {
     bIsDynamicallyLoaded=false
     drawAs=Image
     imageType=NoImage
     mirroring=NoMirror
     resourceName=None
     resourceObject=&/MI_UI_ChargeSMG_AmmoBar
     tiling=NoTile
     imageSize=<80 80>
     margin=<ltrb 0 0 0 0>
     outlineSettings {
      bUseBrushTransparency=false
      roundingType=HalfHeightRadius
      width=0
      color {colorUseRule=UseColor_Specified, specifiedColor=<rgba 0 0 0 0>}
      cornerRadii=<0 0 0 0>
     }
     tintColor {colorUseRule=UseColor_Specified, specifiedColor=<rgba 1 1 1 1>}
     uVRegion {bIsValid=false, max=<0 0>, min=<0 0>}
    }
    colorAndOpacity=<rgba 1 1 1 0.8>
    renderTransform {angle=45, scale=<1 1>, shear=<0 0>, translation=<0 0>}
   }
   slot {class=OverlaySlot, properties {horizontalAlignment=HAlign_Center, verticalAlignment=VAlign_Center, padding=<ltrb 0 0 0 75>}}
  }
  TutorialHighlight {widgetClassPath=&TutorialHighlight}
 ]
}
