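# Translucent UI material for the Charge SMG that draws a ring gauge with progress, delta and background layers in separate colors.
# An angle gradient with start, rotation and inversion controls is thresholded by Ceil into hard progress and delta masks, flippable for countdown.
# Circle outline and layer masks drive color and opacity, with optional radial segments subtracted from opacity.
# Static switches toggle background, delta and segments; Segment Rotation and Edge Softness parameters are unconnected.
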Material /DaisyWeaponGameplay/Items/Weapons/ChargeSMG/UI/M_UI_RadialProgressAndDelta
graph {
 properties {blendMode=BLEND_Translucent, materialDomain=MD_UI, pixelDepthOffsetMode=PDOM_Legacy, usedShadingModels=Default Lit}
 nodes [
  n1 = VectorParameter() {group=4 - Background, parameterName=BG Color, sortPriority=1, defaultValue=<rgba 0.0469 0.0469 0.0469 1>}
  n2 = StaticSwitchParameter(True=Multiply(A=@n1.RGB, B=NamedRerouteUsage()), False=Constant()) {defaultValue=true, group=4 - Background, parameterName=Use BG, sortPriority=0}
  n3 = VectorParameter() {group=2 - Delta, parameterName=Delta Color, sortPriority=2, defaultValue=<rgba 1 0 0.0636 1>}
  n4 = NamedRerouteUsage()
  LinearInterpolate(
   A=StaticSwitchParameter(True=LinearInterpolate(A=@n2, B=@n3.RGB, Alpha=@n4) {constAlpha=0}, False=@n2) {defaultValue=true, group=2 - Delta, parameterName=Use Delta, sortPriority=0}
   B=VectorParameter() {group=1- Progress, parameterName=Progress Color, defaultValue=<rgba 1 1 1 1>}.RGB
   Alpha=NamedRerouteUsage()
  ) {constAlpha=1}
  n5 = StaticSwitchParameter(True=NamedRerouteUsage(), False=Constant()) {defaultValue=true, group=4 - Background, parameterName=Use BG, sortPriority=0}
  n6 = NamedRerouteUsage()
  n7 = Add(
   A=StaticSwitchParameter(True=LinearInterpolate(A=@n5, B=@n6, Alpha=@n6), False=@n5) {defaultValue=true, group=2 - Delta, parameterName=Use Delta, sortPriority=0}
   B=NamedRerouteUsage()
  )
  n8 = NamedRerouteUsage()
  Multiply(
   A=StaticSwitchParameter(True=Subtract(A=@n7, B=@n8), False=@n7) {defaultValue=true, group=3 - Circle Settings, parameterName=Use Segments, sortPriority=4}
   B=ScalarParameter() {defaultValue=1, parameterName=Global Opacity, sliderMax=1}
  )
  n9 = MaterialFunctionCall!(UVs=NamedRerouteUsage(), Rotation=ScalarParameter() {group=1- Progress, parameterName=Progress Rotation, sliderMax=1, sortPriority=1}) {materialFunction=/FortUILibrary/MaterialFunctions/Gradients/MF_UI_AngleGradient}
  n10 = OneMinus(None=@n9.Result)
  n11 = ScalarParameter() {defaultValue=0.2, group=1- Progress, parameterName=Start Value, sliderMax=1, sortPriority=2}
  n12 = NamedRerouteDeclaration(None=Subtract(A=StaticSwitchParameter(True=@n10, False=@n9.Result) {group=1- Progress, parameterName=Inverse Progress, sortPriority=4}, B=@n11)) {name=Progress Gradient, nodeColor=<rgba 0 0 0 1>}
  n13 = Subtract(B=@n11)
  n14 = Multiply(A=ScalarParameter() {defaultValue=0.336, group=1- Progress, parameterName=Progress, sliderMax=1, sortPriority=0}, B=@n13)
  n15 = Ceil(None=Saturate(None=Subtract(A=@n12, B=@n14)))
  NamedRerouteDeclaration(None=StaticSwitchParameter(True=@n15, False=OneMinus(None=@n15)) {group=1- Progress, parameterName=Is Countdown, sortPriority=3}) {name=Progress Mask, nodeColor=<rgba 0 0 0 1>}
  n16 = Subtract(A=MaterialFunctionCall() {materialFunction=&GetUserInterfaceUV}.Normalized UV, B=NamedRerouteUsage())
  n17 = ScalarParameter() {defaultValue=10, group=3 - Circle Settings, parameterName=Segments, sortPriority=5}
  n18 = ScalarParameter() {defaultValue=0.8, group=3 - Circle Settings, parameterName=Segment Spacing, sliderMax=1, sortPriority=6}
  NamedRerouteDeclaration(None=Multiply(
    A=MaterialFunctionCall!(UV=@n16, Segment Count=@n17, Segment Spacing=@n18, Fixed Spacing) {materialFunction=/FortUILibrary/MaterialFunctions/Basics/MF_UI_RadialSegments}.Segmented Triangles Fill
    B=ScalarParameter() {defaultValue=1, group=3 - Circle Settings, parameterName=SegmentOverride, sliderMax=1}
   )) {name=Segments, nodeColor=<rgba 0 0 0 1>}
  n19 = NamedRerouteUsage()
  n20 = Multiply(A=@n13, B=ScalarParameter() {defaultValue=0.6, group=2 - Delta, parameterName=Delta Progress, sliderMax=1, sortPriority=1})
  n21 = Ceil(None=Saturate(None=Subtract(A=@n19, B=@n20)))
  NamedRerouteDeclaration(None=StaticSwitchParameter(True=@n21, False=OneMinus(None=@n21)) {group=1- Progress, parameterName=Is Countdown, sortPriority=3}) {name=Delta Mask, nodeColor=<rgba 0 0 0 1>}
  NamedRerouteDeclaration(None=Multiply(A=NamedRerouteUsage(), B=NamedRerouteUsage())) {name=Progress, nodeColor=<rgba 0 0 0 1>}
  n22 = AppendVector(
   A=ScalarParameter() {group=3 - Circle Settings, parameterName=Offset X}
   B=ScalarParameter() {group=3 - Circle Settings, parameterName=Offset Y}
  )
  NamedRerouteDeclaration(None=Subtract(
    A=MaterialFunctionCall() {materialFunction=&GetUserInterfaceUV}.Normalized UV
    B=NamedRerouteDeclaration(None=@n22) {name=Offset, nodeColor=<rgba 0 0 0 1>}
   )) {name=UVs, nodeColor=<rgba 0 0 0 1>}
  n23 = NamedRerouteUsage()
  n24 = NamedRerouteUsage()
  n25 = NamedRerouteDeclaration(None=Multiply(A=OneMinus(None=@n24), B=ScalarParameter() {defaultValue=0.25, group=4 - Background, parameterName=BG Opacity, sliderMax=1, sortPriority=2})) {name=BG Color Mask, nodeColor=<rgba 0 0 0 1>}
  NamedRerouteDeclaration(None=Saturate(None=Multiply(A=@n23, B=@n25))) {name=BG Opacity, nodeColor=<rgba 0 0 0 1>}
  n26 = NamedRerouteUsage()
  n27 = NamedRerouteUsage()
  n28 = NamedRerouteUsage()
  n29 = NamedRerouteDeclaration(None=Multiply(
    A=Subtract(A=@n27, B=@n28)
    B=ScalarParameter() {defaultValue=0.5, group=2 - Delta, parameterName=Delta Opacity, sliderMax=1, sortPriority=3}
   )) {name=Delta Color Mask, nodeColor=<rgba 0 0 0 1>}
  NamedRerouteDeclaration(None=Saturate(None=Multiply(A=@n26, B=@n29))) {name=Delta Opacity, nodeColor=<rgba 0 0 0 1>}
  NamedRerouteDeclaration(None=MaterialFunctionCall(
    UVs=NamedRerouteUsage()
    Size=ScalarParameter() {defaultValue=0.8, group=3 - Circle Settings, parameterName=Outer Size, sortPriority=0}
    OutlineThickness=ScalarParameter() {defaultValue=0.05, group=3 - Circle Settings, parameterName=Outline Thickness, sortPriority=1}
    GlowMin=ScalarParameter() {group=3 - Circle Settings, parameterName=Edge Softness Min, sortPriority=2}
    GlowMax=ScalarParameter() {defaultValue=0.02, group=3 - Circle Settings, parameterName=Edge Softness Max, sortPriority=3}
   ) {materialFunction=&MF_UI_Circle}.Outline) {name=Circle Mask, nodeColor=<rgba 0 0 0 1>}
  ScalarParameter() {defaultValue=0.07, group=3 - Circle Settings, parameterName=Segment Edge Softness, sliderMax=1, sortPriority=8}
  ScalarParameter() {group=3 - Circle Settings, parameterName=Segment Rotation, sliderMax=1, sortPriority=7}
 ]
}
